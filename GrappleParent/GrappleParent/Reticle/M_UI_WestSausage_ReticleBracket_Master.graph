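# Translucent UI master material for the WestSausage grapple reticle bracket, with parameter-driven lock-on color, scale and fade.
# Center-scaled UVs sample the reticle texture, whose red channel is thresholded and sharpened into alpha.
# Sine of rotated, unscaled UVs procedurally cuts optional crosshair gaps, sized by LockedOn.
# Disabled and final alpha multipliers apply, then an inverted, rotatable radial mask wipes the result.
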Material /GrappleParent/GrappleParent/Reticle/M_UI_WestSausage_ReticleBracket_Master
graph {
 properties {blendMode=BLEND_Translucent, materialDomain=MD_UI, pixelDepthOffsetMode=PDOM_Legacy, usedShadingModels=Default Lit}
 nodes [
  LinearInterpolate(
   A=VectorParameter() {parameterName=StartColor, defaultValue=<rgba 1 0 0 1>}.RGB
   B=VectorParameter() {parameterName=Color, defaultValue=<rgba 1 0 0 1>}.RGB
   Alpha=ScalarParameter() {parameterName=LockedOn}
  )
  n1 = MaterialFunctionCall(
   UVs=TextureCoordinate()
   Texture Scale=LinearInterpolate(
    A=ScalarParameter() {defaultValue=1, parameterName=ScaleStart}
    B=ScalarParameter() {defaultValue=0.75, parameterName=ScaleStop}
    Alpha=ScalarParameter() {parameterName=AnimateScale}
   )
  ) {materialFunction=&ScaleUVsByCenter}
  n2 = TextureSample(UVs=@n1.UVs) {texture=&/T_UI_WestSuasage_Reticle_128x}
  n3 = ScalarParameter() {defaultValue=0.075, parameterName=Subtract}
  n4 = Multiply(A=Saturate(None=Subtract(A=@n2.R, B=@n3)), B=ScalarParameter() {defaultValue=6, parameterName=EdgeHardness})
  n5 = ScalarParameter() {defaultValue=1, parameterName=AlphaStart}
  n6 = ScalarParameter() {defaultValue=1, parameterName=AlphaStop}
  n7 = ScalarParameter() {parameterName=AnimateAlpha}
  n8 = Saturate(None=Multiply(A=Saturate(None=@n4), B=LinearInterpolate(A=@n5, B=@n6, Alpha=@n7)))
  n9 = TextureCoordinate()
  n10 = ScalarParameter() {defaultValue=0.125, parameterName=RotationGapAngle}
  n11 = MaterialFunctionCall(Float2=Sine(None=MaterialFunctionCall(UVs=@n9, "Rotation Angle (0-1)"=@n10) {materialFunction=&CustomRotator}.Rotated Values) {period=2}) {materialFunction=&BreakOutFloat2Components}
  n12 = LinearInterpolate(
   A=ScalarParameter() {defaultValue=0.95, parameterName=CrosshairGap}
   B=ScalarParameter() {defaultValue=1, parameterName=CrosshairGapLockedOn}
   Alpha=ScalarParameter() {parameterName=LockedOn}
  )
  n13 = Subtract(A=Saturate(None=@n11.R), B=@n12)
  n14 = ScalarParameter() {defaultValue=24, parameterName=CrosshairGapEdge}
  n15 = Subtract(A=Saturate(None=@n11.G), B=@n12)
  n16 = Add(A=Saturate(None=Multiply(A=@n13, B=@n14)), B=Saturate(None=Multiply(A=@n15, B=@n14)))
  n17 = DotProduct(A=@n8, B=OneMinus(None=Saturate(None=@n16)))
  n18 = Multiply(
   A=LinearInterpolate(A=@n8, B=Saturate(None=@n17), Alpha=ScalarParameter() {parameterName=UseCrop})
   B=LinearInterpolate(B=ScalarParameter() {defaultValue=0.5, parameterName=DisabledAlpha}, Alpha=ScalarParameter() {parameterName=AnimateDisabledAlpha}) {constA=1}
  )
  n19 = ScalarParameter() {defaultValue=1, parameterName=Final_Alpha}
  n20 = MaterialFunctionCall(UVs=@n1.UVs, "Rotation Angle (0-1)"=ScalarParameter() {defaultValue=0.125, parameterName=RotateMask}) {materialFunction=&CustomRotator}
  n21 = ScalarParameter() {parameterName=Mask}
  Multiply(
   A=OneMinus(None=MaterialFunctionCall!(UVs=@n20.Rotated Values, EdgeSoftness, Progress=@n21) {materialFunction=/Game/UI/Foundation/Toasts/Materials/MF_RadialMask}.Mask)
   B=Saturate(None=Multiply(A=@n18, B=@n19))
  )
 ]
}
