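# UI master material for targeting reticle circles, with static switches selecting a center dot, a radial cooldown progress ring, or rotating arc rings.
# Radial gradients form dot and ring masks, while thresholded UV angles cut the progress fill and two opposed, sine-animated arcs.
# Optional pulse and looping growth modulate opacity and dot radius; color blends backing to fill or normal to disabled.
# HasCenterDot reads inverted: its False branch adds the dot.
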Material /GrappleParent/GrappleParent/Reticle/M_UI_WestSausage_Reticle_TargetingCircles_Master
graph {
 properties {blendMode=BLEND_Translucent, materialDomain=MD_UI, pixelDepthOffsetMode=PDOM_Legacy, usedShadingModels=Default Lit}
 nodes [
  n1 = ScalarParameter() {defaultValue=0.2, parameterName=CenterDot_Radius}
  n2 = ScalarParameter() {defaultValue=0.2, parameterName=CenterDot_Radius_Stop}
  n3 = Fmod(A=Multiply(A=Time(), B=ScalarParameter() {defaultValue=0.5, parameterName=GrowthRate}), B=Constant() {r=1})
  n4 = MaterialFunctionCall(
   UVs=TextureCoordinate()
   Radius=StaticSwitchParameter(True=LinearInterpolate(A=@n1, B=@n2, Alpha=@n3), False=ScalarParameter() {defaultValue=0.2, parameterName=CenterDot_Radius}) {parameterName=HasGrowth}
   Density=ScalarParameter() {defaultValue=14, parameterName=EdgeHardness_CenterDot}
  ) {materialFunction=&RadialGradientExponential}
  n5 = ScalarParameter() {defaultValue=0.25, group=Section, parameterName=Section_Orientation}
  n6 = TextureCoordinate()
  n7 = Multiply(A=Subtract(
    A=MaterialFunctionCall!(Rotation_Fraction=@n5, TexCoord=@n6) {materialFunction=/FortUILibrary/Foundation/Art_ToolKit/AngledBoxes/MF_UI_Rotator_Fractions}.Result
    B=Constant() {r=0.5}
   )) {constB=2}
  n8 = Subtract(
   A=Saturate(None=MaterialFunctionCall(Vector or UVs=@n7) {materialFunction=&VectorToRadialValue}.Vector Converted to Angle)
   B=LinearInterpolate(A=ScalarParameter() {parameterName=PercentCompensation}, Alpha=ScalarParameter() {defaultValue=0.333, parameterName=CooldownPercent})
  )
  n9 = ScalarParameter() {defaultValue=256, parameterName=Fill_Hardness}
  n10 = OneMinus(None=Saturate(None=Multiply(A=@n8, B=@n9)))
  n11 = ScalarParameter() {defaultValue=0.5, parameterName=OuterRadius}
  n12 = ScalarParameter() {defaultValue=24, parameterName=EdgeHardness_Outer}
  n13 = TextureCoordinate()
  n14 = MaterialFunctionCall(
   UVs=@n13
   Radius=ScalarParameter() {defaultValue=0.4, parameterName=InnerRadius}
   Density=ScalarParameter() {defaultValue=22, parameterName=EdgeHardness_Inner}
  ) {materialFunction=&RadialGradientExponential}
  n15 = Saturate(None=DotProduct(
    A=MaterialFunctionCall(UVs=@n13, Radius=@n11, Density=@n12) {materialFunction=&RadialGradientExponential}.RadialGradientExponential
    B=OneMinus(None=@n14.RadialGradientExponential)
   ))
  n16 = DotProduct(
   A=LinearInterpolate(
    A=ScalarParameter() {defaultValue=0.25, parameterName=Fuel_BackingAlpha}
    B=ScalarParameter() {defaultValue=1, parameterName=Fuel_FillAlpha}
    Alpha=@n10
   )
   B=@n15
  )
  n17 = Multiply(A=Multiply(A=Time()) {constB=-1}, B=ScalarParameter() {defaultValue=0.5, parameterName=RotationSpeed})
  n18 = MaterialFunctionCall(
   Rotation_Fraction=ScalarParameter() {defaultValue=0.25, group=Section, parameterName=Section_Orientation}
   TexCoord=Rotator(Coordinate=TextureCoordinate(), Time=@n17)
  ) {materialFunction=&MF_UI_Rotator_Fractions}
  n19 = Constant() {r=0.5}
  n20 = MaterialFunctionCall(Vector or UVs=Multiply(A=Subtract(A=@n18.Result, B=@n19)) {constB=2}) {materialFunction=&VectorToRadialValue}
  n21 = LinearInterpolate(
   A=@n20.Vector Converted to Angle
   B=OneMinus(None=@n20.Vector Converted to Angle)
   Alpha=ScalarParameter() {parameterName=FlipSubtractionDirection}
  )
  n22 = Sine(None=Multiply(A=Time(), B=ScalarParameter() {defaultValue=0.75, parameterName=SubtractRate}))
  n23 = LinearInterpolate(
   A=ScalarParameter() {defaultValue=0.875, parameterName=Angle_Subtract}
   B=ScalarParameter() {defaultValue=0.666, parameterName=Angle_Subtract_Stop}
   Alpha=Multiply(A=Add(A=@n22)) {constB=0.5}
  )
  n24 = Multiply(A=Subtract(A=Saturate(None=@n21), B=@n23), B=ScalarParameter() {defaultValue=256, parameterName=Angle_Hardness})
  n25 = TextureCoordinate()
  n26 = MaterialFunctionCall(
   Rotation_Fraction=ScalarParameter() {defaultValue=0.25, group=Section, parameterName=Section_Orientation}
   TexCoord=OneMinus(None=Rotator(Coordinate=@n25, Time=@n17))
  ) {materialFunction=&MF_UI_Rotator_Fractions}
  n27 = Constant() {r=0.5}
  n28 = MaterialFunctionCall(Vector or UVs=Multiply(A=Subtract(A=@n26.Result, B=@n27)) {constB=2}) {materialFunction=&VectorToRadialValue}
  n29 = LinearInterpolate(
   A=@n28.Vector Converted to Angle
   B=OneMinus(None=@n28.Vector Converted to Angle)
   Alpha=ScalarParameter() {parameterName=FlipSubtractionDirection}
  )
  n30 = Multiply(A=Subtract(A=Saturate(None=@n29), B=@n23), B=ScalarParameter() {defaultValue=256, parameterName=Angle_Hardness})
  n31 = DotProduct(A=OneMinus(None=Saturate(None=@n24)), B=OneMinus(None=Saturate(None=@n30)))
  n32 = Saturate(None=DotProduct(A=@n15, B=Saturate(None=@n31)))
  n33 = ScalarParameter() {defaultValue=0.5, parameterName=RingsAlpha}
  n34 = Saturate(None=Add(A=Multiply(A=@n32, B=@n33), B=@n4.RadialGradientExponential))
  n35 = StaticSwitchParameter(
   True=@n4.RadialGradientExponential
   False=StaticSwitchParameter(True=Saturate(None=@n16), False=StaticSwitchParameter(True=@n32, False=@n34) {parameterName=HasCenterDot}) {parameterName=IsProgressRing}
  ) {parameterName=IsJustCenterDot}
  n36 = Sine(None=Multiply(A=Time(), B=ScalarParameter() {defaultValue=2, parameterName=PulseRate}))
  n37 = LinearInterpolate(
   A=ScalarParameter() {defaultValue=0.333, parameterName=PulseAlpha_Start}
   B=ScalarParameter() {defaultValue=0.666, parameterName=PulseAlpha_Stop}
   Alpha=Multiply(A=Add(A=@n36)) {constB=0.5}
  )
  n38 = StaticSwitchParameter(True=Saturate(None=Multiply(A=@n35, B=@n37)), False=@n35) {parameterName=HasPulse}
  n39 = ScalarParameter() {defaultValue=1, parameterName=GrowthAlpha_Start}
  n40 = ScalarParameter() {parameterName=GrowthAlpha_Stop}
  n41 = Fmod(A=Multiply(A=Time(), B=ScalarParameter() {defaultValue=0.5, parameterName=GrowthRate}), B=Constant() {r=1})
  n42 = Saturate(None=Multiply(A=@n38, B=LinearInterpolate(A=@n39, B=@n40, Alpha=@n41)))
  Saturate(None=Multiply(A=StaticSwitchParameter(True=@n42, False=@n38) {parameterName=HasGrowth}, B=ScalarParameter() {defaultValue=1, parameterName=FinalAlpha}))
  ScalarParameter() {defaultValue=0.333, parameterName=PulseAlpha_Start}
  ScalarParameter() {defaultValue=0.666, parameterName=PulseAlpha_Stop}
  ScalarParameter() {defaultValue=0.875, parameterName=Angle_Subtract}
  ScalarParameter() {defaultValue=0.875, parameterName=Angle_Subtract}
  n43 = VectorParameter() {parameterName=Color, defaultValue=<rgba 0 1 0 1>}
  n44 = VectorParameter() {parameterName=BackingColor, defaultValue=<rgba 0 0 1 1>}
  n45 = VectorParameter() {parameterName=Color, defaultValue=<rgba 0 1 0 1>}
  n46 = VectorParameter() {parameterName=Color_Disabled, defaultValue=<rgba 1 0 0 1>}
  n47 = ScalarParameter() {parameterName=IsDisabled}
  StaticSwitchParameter(
   True=@n43.RGB
   False=StaticSwitchParameter(True=LinearInterpolate(A=@n44.RGB, B=@n45.RGB, Alpha=@n10), False=LinearInterpolate(A=@n43.RGB, B=@n46.RGB, Alpha=@n47)) {parameterName=IsProgressRing}
  ) {parameterName=IsJustCenterDot}
 ]
}
