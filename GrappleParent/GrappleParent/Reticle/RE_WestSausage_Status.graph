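# Reticle extension widget for the grapple weapon, showing lock-on, swing-attached, firing-blocked and cooldown states.
# On construct it casts the equipped weapon to B_Weapon_GrappleParent and binds its reticle update and cooldown events.
# Updates tint bracket images red when firing is blocked and play flag-guarded lock-on and swing animations forward or reversed.
# When unusable, including in vehicles, the bracket hides; during cooldown a server-time-driven material ring appears.
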WidgetBlueprint /GrappleParent/GrappleParent/Reticle/RE_WestSausage_Status
animations [
 Swinging {
  binding_count=4
  length_seconds=0.2
  object=&self:Swinging
  sequence {bindings [
    Reticle_Surrounding_BaseDarken {
     id=#1
     possessed_class=Image
     tracks [{
       type=MovieSceneWidgetMaterialTrack
       sections [{
         type=MovieSceneParameterSection
         channels [
          AnimateScale {type=MovieSceneScriptingFloatChannel, keys [[0,0],[4,1]]}
          CrosshairGap {type=MovieSceneScriptingFloatChannel, const [1,1]}
         ]
        }]
      }]
    }
    Reticle_Surrounding_Shadow {
     id=#2
     possessed_class=Image
     tracks [{
       type=MovieSceneWidgetMaterialTrack
       sections [{
         type=MovieSceneParameterSection
         channels [
          AnimateScale {type=MovieSceneScriptingFloatChannel, keys [[0,0],[4,1]]}
          CrosshairGap {type=MovieSceneScriptingFloatChannel, const [1,1]}
         ]
        }]
      }]
    }
    Reticle_Spinners_AttachedFast {
     id=#3
     possessed_class=Image
     tracks [
      {
       type=MovieScene2DTransformTrack
       sections [{
         type=MovieScene2DTransformSection
         channels [
          Translation.X {type=MovieSceneScriptingFloatChannel, const [0,1]}
          Translation.Y {type=MovieSceneScriptingFloatChannel, const [0,1]}
          Angle {type=MovieSceneScriptingFloatChannel, const [0,1]}
          Scale.X {type=MovieSceneScriptingFloatChannel, keys [[0,1],[4,0.75]]}
          Scale.Y {type=MovieSceneScriptingFloatChannel, keys [[0,1],[4,0.75]]}
          Shear.X {type=MovieSceneScriptingFloatChannel, const [0,1]}
          Shear.Y {type=MovieSceneScriptingFloatChannel, const [0,1]}
         ]
        }]
      }
      {
       type=MovieSceneFloatTrack
       display_name=Render Opacity
       sections [{type=MovieSceneFloatSection, channels [None {type=MovieSceneScriptingFloatChannel, keys [[0,0],[2,1]]}]}]
      }
     ]
    }
    Reticle_Surrounding {
     id=#4
     possessed_class=Image
     tracks [{
       type=MovieSceneWidgetMaterialTrack
       sections [{
         type=MovieSceneParameterSection
         channels [
          AnimateScale {type=MovieSceneScriptingFloatChannel, keys [[0,0],[4,1]]}
          AnimateAlpha {type=MovieSceneScriptingFloatChannel, keys [[0,0],[4,1]]}
          CrosshairGap {type=MovieSceneScriptingFloatChannel, const [1,1]}
         ]
        }]
      }]
    }
   ]}
 }
 CantUse {
  binding_count=5
  length_seconds=0.2
  object=&self:CantUse
  sequence {bindings [
    AnimBoundSizeControl {
     id=#5
     possessed_class=Border
     tracks [{
       type=MovieScene2DTransformTrack
       sections [{
         type=MovieScene2DTransformSection
         channels [
          Translation.X {type=MovieSceneScriptingFloatChannel, const [0,1]}
          Translation.Y {type=MovieSceneScriptingFloatChannel, const [0,1]}
          Angle {type=MovieSceneScriptingFloatChannel, keys [[0,0],[4,45]]}
          Scale.X {type=MovieSceneScriptingFloatChannel, keys [[0,1],[4,0.75]]}
          Scale.Y {type=MovieSceneScriptingFloatChannel, keys [[0,1],[4,0.75]]}
          Shear.X {type=MovieSceneScriptingFloatChannel, const [0,1]}
          Shear.Y {type=MovieSceneScriptingFloatChannel, const [0,1]}
         ]
        }]
      }]
    }
    Reticle_Surrounding_BaseDarken {
     id=#6
     possessed_class=Image
     tracks [{
       type=MovieSceneWidgetMaterialTrack
       sections [{
         type=MovieSceneParameterSection
         channels [
          CrosshairGap {type=MovieSceneScriptingFloatChannel, keys [[0,0.885],[4,1]]}
          AnimateScale {type=MovieSceneScriptingFloatChannel, const [0,1]}
         ]
        }]
      }]
    }
    Reticle_Surrounding {
     id=#7
     possessed_class=Image
     tracks [{
       type=MovieSceneWidgetMaterialTrack
       sections [{
         type=MovieSceneParameterSection
         channels [
          CrosshairGap {type=MovieSceneScriptingFloatChannel, keys [[0,0.885],[4,1]]}
          AnimateDisabledAlpha {type=MovieSceneScriptingFloatChannel, const [1,1]}
          AnimateScale {type=MovieSceneScriptingFloatChannel, const [0,1]}
         ]
        }]
      }]
    }
    Reticle_Surrounding_Disabled {
     id=#8
     possessed_class=Image
     tracks [{
       type=MovieSceneWidgetMaterialTrack
       sections [{
         type=MovieSceneParameterSection
         channels [
          CrosshairGap {type=MovieSceneScriptingFloatChannel, keys [[0,0.885],[4,1]]}
          AnimateScale {type=MovieSceneScriptingFloatChannel, const [0,1]}
         ]
        }]
      }]
    }
    Reticle_Surrounding_Shadow {
     id=#9
     possessed_class=Image
     tracks [{
       type=MovieSceneWidgetMaterialTrack
       sections [{
         type=MovieSceneParameterSection
         channels [
          CrosshairGap {type=MovieSceneScriptingFloatChannel, keys [[0,0.885],[4,1]]}
          AnimateScale {type=MovieSceneScriptingFloatChannel, const [0,1]}
         ]
        }]
      }]
    }
   ]}
 }
 LockedOn {
  binding_count=4
  length_seconds=0.1
  object=&self:LockedOn
  sequence {bindings [
    Reticle_Surrounding {
     id=#10
     possessed_class=Image
     tracks [{
       type=MovieSceneWidgetMaterialTrack
       sections [{
         type=MovieSceneParameterSection
         channels [
          CrosshairGap {type=MovieSceneScriptingFloatChannel, keys [[0,0.885],[2,1]]}
          AnimateDisabledAlpha {type=MovieSceneScriptingFloatChannel, keys [[0,1],[2,0]]}
          LockedOn {type=MovieSceneScriptingFloatChannel, keys [[0,0],[2,1]]}
         ]
        }]
      }]
    }
    Reticle_Surrounding_Shadow {
     id=#11
     possessed_class=Image
     tracks [{
       type=MovieSceneWidgetMaterialTrack
       sections [{type=MovieSceneParameterSection, channels [CrosshairGap {type=MovieSceneScriptingFloatChannel, keys [[0,0.885],[2,1]]}]}]
      }]
    }
    Bracket {
     id=#12
     possessed_class=Overlay
     tracks [
      {
       type=MovieScene2DTransformTrack
       sections [{
         type=MovieScene2DTransformSection
         channels [
          Scale.X {type=MovieSceneScriptingFloatChannel, keys [[0,1.25],[2,1]]}
          Scale.Y {type=MovieSceneScriptingFloatChannel, keys [[0,1.25],[2,1]]}
         ]
        }]
      }
      {
       type=MovieSceneFloatTrack
       display_name=Render Opacity
       sections [{type=MovieSceneFloatSection, channels [None {type=MovieSceneScriptingFloatChannel, keys [[0,0.5],[2,1]]}]}]
      }
     ]
    }
    Reticle_Surrounding_BaseDarken {
     id=#13
     possessed_class=Image
     tracks [{
       type=MovieSceneWidgetMaterialTrack
       sections [{type=MovieSceneParameterSection, channels [CrosshairGap {type=MovieSceneScriptingFloatChannel, keys [[0,0.885],[2,1]]}]}]
      }]
    }
   ]}
 }
 Cooldown_Intro {
  binding_count=1
  length_seconds=0.4
  object=&self:Cooldown_Intro
  sequence {bindings [CooldownRing {
     id=#14
     possessed_class=Image
     tracks [
      {
       type=MovieScene2DTransformTrack
       sections [{
         type=MovieScene2DTransformSection
         channels [
          Scale.X {type=MovieSceneScriptingFloatChannel, keys [[0,0.5],[8,1]]}
          Scale.Y {type=MovieSceneScriptingFloatChannel, keys [[0,0.5],[8,1]]}
         ]
        }]
      }
      {
       type=MovieSceneFloatTrack
       display_name=Render Opacity
       sections [{type=MovieSceneFloatSection, channels [None {type=MovieSceneScriptingFloatChannel, keys [[0,0],[5,1]]}]}]
      }
     ]
    }]}
 }
 LockedOnNew {
  binding_count=6
  length_seconds=0.1
  object=&self:LockedOnNew
  sequence {bindings [
    Reticle_Surrounding {
     id=#15
     possessed_class=Image
     tracks [{
       type=MovieSceneWidgetMaterialTrack
       sections [{
         type=MovieSceneParameterSection
         channels [
          LockedOn {type=MovieSceneScriptingFloatChannel, keys [[0,0],[2,1]]}
          AnimateAlpha {type=MovieSceneScriptingFloatChannel, keys [[0,0],[2,1]]}
         ]
        }]
      }]
    }
    Reticle_Surrounding_Shadow {
     id=#16
     possessed_class=Image
     tracks [{
       type=MovieSceneWidgetMaterialTrack
       sections [{type=MovieSceneParameterSection, channels [LockedOn {type=MovieSceneScriptingFloatChannel, keys [[0,0],[2,1]]}]}]
      }]
    }
    Reticle_Surrounding_BaseDarken {
     id=#17
     possessed_class=Image
     tracks [{
       type=MovieSceneWidgetMaterialTrack
       sections [{type=MovieSceneParameterSection, channels [LockedOn {type=MovieSceneScriptingFloatChannel, keys [[0,0],[2,1]]}]}]
      }]
    }
    Reticle_Spinners_AttachedFast {
     id=#18
     possessed_class=Image
     tracks [{
       type=MovieSceneFloatTrack
       display_name=Render Opacity
       sections [{type=MovieSceneFloatSection, channels [None {type=MovieSceneScriptingFloatChannel, const [0,1]}]}]
      }]
    }
    AnimBoundSizeControl {
     id=#19
     possessed_class=Border
     tracks [{
       type=MovieScene2DTransformTrack
       sections [{
         type=MovieScene2DTransformSection
         channels [
          Scale.X {type=MovieSceneScriptingFloatChannel, keys [[0,1.1875],[2,1]]}
          Scale.Y {type=MovieSceneScriptingFloatChannel, keys [[0,1.1875],[2,1]]}
         ]
        }]
      }]
    }
    Bracket {
     id=#20
     possessed_class=Overlay
     tracks [{
       type=MovieScene2DTransformTrack
       sections [{
         type=MovieScene2DTransformSection
         channels [
          Angle {type=MovieSceneScriptingFloatChannel, const [0,1]}
          Scale.X {type=MovieSceneScriptingFloatChannel, const [1,1]}
          Scale.Y {type=MovieSceneScriptingFloatChannel, const [1,1]}
         ]
        }]
      }]
    }
   ]}
 }
 SwingingNew {
  binding_count=6
  length_seconds=0.2
  object=&self:SwingingNew
  sequence {bindings [
    Reticle_Surrounding {
     id=#15
     possessed_class=Image
     tracks [{
       type=MovieSceneWidgetMaterialTrack
       sections [{
         type=MovieSceneParameterSection
         channels [
          LockedOn {type=MovieSceneScriptingFloatChannel, const [1,1]}
          AnimateAlpha {type=MovieSceneScriptingFloatChannel, const [1,1]}
         ]
        }]
      }]
    }
    Reticle_Surrounding_BaseDarken {
     id=#21
     possessed_class=Image
     tracks [{
       type=MovieSceneWidgetMaterialTrack
       sections [{type=MovieSceneParameterSection, channels [LockedOn {type=MovieSceneScriptingFloatChannel, const [1,1]}]}]
      }]
    }
    Bracket {
     id=#22
     possessed_class=Overlay
     tracks [{
       type=MovieScene2DTransformTrack
       sections [{
         type=MovieScene2DTransformSection
         channels [
          Angle {type=MovieSceneScriptingFloatChannel, keys [[0,0],[4,45]]}
          Scale.X {type=MovieSceneScriptingFloatChannel, keys [[0,1],[4,0.75]]}
          Scale.Y {type=MovieSceneScriptingFloatChannel, keys [[0,1],[4,0.75]]}
         ]
        }]
      }]
    }
    AnimBoundSizeControl {
     id=#23
     possessed_class=Border
     tracks [{
       type=MovieScene2DTransformTrack
       sections [{
         type=MovieScene2DTransformSection
         channels [
          Scale.X {type=MovieSceneScriptingFloatChannel, const [1,1]}
          Scale.Y {type=MovieSceneScriptingFloatChannel, const [1,1]}
         ]
        }]
      }]
    }
    Reticle_Spinners_AttachedFast {
     id=#24
     possessed_class=Image
     tracks [{
       type=MovieSceneFloatTrack
       display_name=Render Opacity
       sections [{type=MovieSceneFloatSection, channels [None {type=MovieSceneScriptingFloatChannel, keys [[0,0],[4,1]]}]}]
      }]
    }
    Reticle_Surrounding_Shadow {
     id=#25
     possessed_class=Image
     tracks [{
       type=MovieSceneWidgetMaterialTrack
       sections [{type=MovieSceneParameterSection, channels [LockedOn {type=MovieSceneScriptingFloatChannel, const [1,1]}]}]
      }]
    }
   ]}
 }
 CantUseNew {
  binding_count=6
  length_seconds=0.1
  object=&self:CantUseNew
  sequence {bindings [
    Reticle_Surrounding {
     id=#15
     possessed_class=Image
     tracks [{
       type=MovieSceneWidgetMaterialTrack
       sections [{
         type=MovieSceneParameterSection
         channels [
          LockedOn {type=MovieSceneScriptingFloatChannel, keys [[0,1],[2,0]]}
          AnimateAlpha {type=MovieSceneScriptingFloatChannel, const [1,1]}
         ]
        }]
      }]
    }
    Reticle_Surrounding_Shadow {
     id=#16
     possessed_class=Image
     tracks [{
       type=MovieSceneWidgetMaterialTrack
       sections [{type=MovieSceneParameterSection, channels [LockedOn {type=MovieSceneScriptingFloatChannel, keys [[0,1],[2,0]]}]}]
      }]
    }
    Reticle_Surrounding_BaseDarken {
     id=#17
     possessed_class=Image
     tracks [{
       type=MovieSceneWidgetMaterialTrack
       sections [{type=MovieSceneParameterSection, channels [LockedOn {type=MovieSceneScriptingFloatChannel, keys [[0,1],[2,0]]}]}]
      }]
    }
    Reticle_Spinners_AttachedFast {
     id=#18
     possessed_class=Image
     tracks [{
       type=MovieSceneFloatTrack
       display_name=Render Opacity
       sections [{type=MovieSceneFloatSection, channels [None {type=MovieSceneScriptingFloatChannel}]}]
      }]
    }
    AnimBoundSizeControl {
     id=#19
     possessed_class=Border
     tracks [{
       type=MovieScene2DTransformTrack
       sections [{
         type=MovieScene2DTransformSection
         channels [
          Scale.X {type=MovieSceneScriptingFloatChannel, const [1,2]}
          Scale.Y {type=MovieSceneScriptingFloatChannel, const [1,1]}
         ]
        }]
      }]
    }
    Bracket {
     id=#20
     possessed_class=Overlay
     tracks [{
       type=MovieScene2DTransformTrack
       sections [{
         type=MovieScene2DTransformSection
         channels [
          Angle {type=MovieSceneScriptingFloatChannel, const [0,1]}
          Scale.X {type=MovieSceneScriptingFloatChannel, const [1,1]}
          Scale.Y {type=MovieSceneScriptingFloatChannel, const [1,1]}
         ]
        }]
      }]
    }
   ]}
 }
 Intro {
  binding_count=1
  length_seconds=0.3
  object=&self:Intro
  sequence {bindings [Reticle {
     id=#26
     possessed_class=Overlay
     tracks [
      {
       type=MovieSceneFloatTrack
       display_name=Render Opacity
       sections [{type=MovieSceneFloatSection, channels [None {type=MovieSceneScriptingFloatChannel, keys [[0,0],[6,1]]}]}]
      }
      {
       type=MovieScene2DTransformTrack
       sections [{
         type=MovieScene2DTransformSection
         channels [
          Translation.X {type=MovieSceneScriptingFloatChannel, const [0,1]}
          Translation.Y {type=MovieSceneScriptingFloatChannel, const [0,1]}
          Angle {type=MovieSceneScriptingFloatChannel, const [0,1]}
          Scale.X {type=MovieSceneScriptingFloatChannel, keys [[0,1.25],[6,1]]}
          Scale.Y {type=MovieSceneScriptingFloatChannel, keys [[0,1.25],[6,1]]}
          Shear.X {type=MovieSceneScriptingFloatChannel, const [0,1]}
          Shear.Y {type=MovieSceneScriptingFloatChannel, const [0,1]}
         ]
        }]
      }
     ]
    }]}
 }
]
graphs [
 EventGraph {nodes [
   n1 = ReticleCooldown!() -> (OutputDelegate, then, Cooldown Duration:double=~, NormalCooldownDuration:double=~) {is=title K2Node_CustomEvent}
   > Set TotalCooldownDuration(TotalCooldownDuration:double=@n1.Cooldown Duration) -> (Output_Get:double)
   > Set FullCooldown(FullCooldown:double=@n1.NormalCooldownDuration) -> (Output_Get:double)
   > Set CooldownStartTime(CooldownStartTime:double=GetServerWorldTimeSeconds!() -> (ReturnValue:float)) -> (Output_Get:double)
   n2 = BP_ReticleUpdate!() -> (
    OutputDelegate
    then
    IsLockedOn:bool=~
    Position:Vector=~
    AbleToUse:bool=~
    ShowChargeUpCircle:bool=~
    ChargeTime:double=~
    IsTargetAtAimLocation:bool=~
    IsSwingAttached:bool=~
    IsFiringBlocked:bool=~
   ) {is=title K2Node_CustomEvent}
   > Update!(execute, self, Position:"Vector (by ref)"=@n2.Position, AbleToUse:bool=@n2.AbleToUse) -> (then)
   > SetLockedOnState!(execute, self, LockedOn:bool=@n2.IsLockedOn, Attached:bool=@n2.IsSwingAttached, IsFiringBlocked:bool=@n2.IsFiringBlocked) -> (then)
   n3 = Update!() -> (OutputDelegate, then, Position:"Vector (by ref)", AbleToUse:bool=~) {is=title K2Node_CustomEvent}
   > Set TargetPosition(TargetPosition:Vector=@n3.Position) -> (Output_Get:Vector)
   > n4 = HasAnyMatchingGameplayTags!(
    execute:Exec
    self:Object Reference=GetOwningPlayerPawn!(self:User Widget Object Reference) -> (ReturnValue:Pawn Object Reference)
    TagContainer:"Gameplay Tag Container Structure (by ref)"=MakeLiteralGameplayTagContainer!(Value:Gameplay Tag Container Structure=(GameplayTags=((TagName="Gameplay.Action.Player.AthenaInVehicle")))) -> (ReturnValue:Gameplay Tag Container Structure)
   ) -> (then:Exec, ReturnValue:bool) {is=title K2Node_Message}
   > n5 = Set OnCooldown(OnCooldown:bool=NOT Boolean(A=@n3.AbleToUse)) -> (Output_Get:bool)
   > Set AbleToUse(AbleToUse:bool=AND Boolean(A=NOT Boolean(A=@n4), B=@n3.AbleToUse)) -> (Output_Get:bool)
   > Branch(Condition=@n5)
   > CooldownTick()
   CooldownTick() {is=title K2Node_CustomEvent}
   > n6 = Set CurrentCooldownDuration(CurrentCooldownDuration:double=float - float(A=GetServerWorldTimeSeconds(), B=Get CooldownStartTime() -> (CooldownStartTime:double))) -> (Output_Get:double)
   > n7 = Branch(Condition="float >= float"(A=@n6, B=Get TotalCooldownDuration() -> (TotalCooldownDuration:double)))
   n8 = Get FullCooldown() -> (FullCooldown:double)
   n9 = Get TotalCooldownDuration()
   n10 = "Max (Float)"(A=float - float(A=@n8, B=@n9), B=0)
   n11 = Get CurrentCooldownDuration() -> (CurrentCooldownDuration:double)
   SetScalarParameterValue(
    execute=@n7.else
    self=Get CooldownRingDynamicMaterial() -> (CooldownRingDynamicMaterial:Material Instance Dynamic Object Reference)
    ParameterName=CooldownPercent
    Value=float / float(A=float + float(A=@n11, B=@n10), B="Max (Float)"(A=@n8, B=@n9))
   )
   n12 = SetLockedOnState!() -> (OutputDelegate, then, LockedOn:bool=~, Attached:bool=~, IsFiringBlocked:bool=~) {is=title K2Node_CustomEvent}
   > Set Locked On(Locked On:bool=@n12.LockedOn) -> (Output_Get:bool)
   > Set Attached(Attached:bool=@n12.Attached) -> (Output_Get:bool)
   > Set IsFiringBlocked(IsFiringBlocked:bool=@n12.IsFiringBlocked) -> (Output_Get:bool)
   > SetColorAndOpacity(
    self:Image Object Reference=[
     Get Reticle_Surrounding_Disabled() -> (Reticle_Surrounding_Disabled:Image Object Reference)
     Get Reticle_Surrounding_Shadow() -> (Reticle_Surrounding_Shadow:Image Object Reference)
     Get Reticle_Spinners_AttachedFast() -> (Reticle_Spinners_AttachedFast:Image Object Reference)
     Get Reticle_Surrounding() -> (Reticle_Surrounding:Image Object Reference)
     Get Reticle_Surrounding_BaseDarken() -> (Reticle_Surrounding_BaseDarken:Image Object Reference)
    ]
    InColorAndOpacity:Linear Color Structure=SelectColor(A=(R=1,G=0,B=0,A=0.5), B=(R=1,G=1,B=1,A=1), bPickA=Get IsFiringBlocked() -> (IsFiringBlocked:bool))
   )
   > n13 = Branch(Condition=Get AbleToUse() -> (AbleToUse:bool))
   n14 = Get Bracket() -> (Bracket:Overlay Object Reference)
   > SetVisibility(self=@n14, InVisibility=SelfHitTestInvisible)
   n15 = Get CooldownRing() -> (CooldownRing:Image Object Reference)
   > SetVisibility(self=@n15, InVisibility=Collapsed)
   > n16 = Branch(Condition=Get IsCooldownPlaying() -> (IsCooldownPlaying:bool))
   > n17 = PlayAnimation(InAnimation=Get Intro())
   Set IsCooldownPlaying(execute=[@n17.then, @n16.else], IsCooldownPlaying:bool=false) -> (Output_Get:bool)
   n18 = NOT Boolean(A=Get IsFiringBlocked())
   n19 = Get Attached() -> (Attached:bool)
   > n20 = Branch(Condition=AND Boolean(A=Get Locked On() -> (Locked On:bool), B=OR Boolean(A=@n18, B=@n19)))
   > n21 = Branch(Condition=Get HasPlayedLockedOnAnim() -> (HasPlayedLockedOnAnim:bool))
   n22 = Branch(execute=[@n24.then, @n21.then], Condition=Get Attached())
   > SetVisibility(self=Get Reticle_Spinners_AttachedFast(), InVisibility=SelfHitTestInvisible)
   > n23 = Branch(Condition=Get HasPlayedAttachedAnim() -> (HasPlayedAttachedAnim:bool))
   PlayAnimationForward(execute=@n23.else, InAnimation=Get SwingingNew() -> (SwingingNew:Widget Animation Object Reference))
   > Set HasPlayedAttachedAnim(HasPlayedAttachedAnim:bool=true) -> (Output_Get:bool)
   SetVisibility(execute=@n22.else, self=Get Reticle_Spinners_AttachedFast(), InVisibility=Collapsed)
   > Branch(Condition=Get HasPlayedAttachedAnim())
   > PlayAnimationReverse(InAnimation=Get SwingingNew())
   > Set HasPlayedAttachedAnim(HasPlayedAttachedAnim=false)
   PlayAnimationForward(execute=@n21.else, InAnimation=Get LockedOnNew() -> (LockedOnNew:Widget Animation Object Reference))
   > n24 = Set HasPlayedLockedOnAnim(HasPlayedLockedOnAnim:bool=true) -> (Output_Get:bool)
   n25 = Branch(execute=@n20.else, Condition=Get Attached())
   SetVisibility(execute=@n25.else, self=Get Reticle_Spinners_AttachedFast(), InVisibility=Collapsed)
   > Branch(Condition=Get HasPlayedLockedOnAnim())
   > PlayAnimationReverse(InAnimation=Get LockedOnNew())
   > Set HasPlayedLockedOnAnim(HasPlayedLockedOnAnim=false)
   SetVisibility(execute=@n13.else, self=@n14, InVisibility=Collapsed)
   > SetVisibility(self=@n15, InVisibility=Select(Option 0=Collapsed, Option 1=SelfHitTestInvisible, Index=Get OnCooldown() -> (OnCooldown:bool)))
   > SetVisibility(self=Get Reticle_Spinners_AttachedFast(), InVisibility=Collapsed)
   > n26 = Branch(Condition=Get IsCooldownPlaying())
   Set IsCooldownPlaying(execute=@n26.else, IsCooldownPlaying=true)
   > PlayAnimation(InAnimation=Get Cooldown_Intro() -> (Cooldown_Intro:Widget Animation Object Reference))
   n27 = Get LockedOnNew()
   > PlayAnimation(InAnimation=@n27, StartAtTime=GetEndTime(self=@n27), PlayMode=Reverse)
   Event Construct()
   > SetVisibility(self=Get Bracket(), InVisibility=Collapsed)
   > SetVisibility(self=Get CooldownRing(), InVisibility=Collapsed)
   > n28 = GetDynamicMaterial(self=Get CooldownRing())
   > Set CooldownRingDynamicMaterial(CooldownRingDynamicMaterial:Material Instance Dynamic Object Reference=@n28) -> (Output_Get:Material Instance Dynamic Object Reference)
   > n29 = Cast To B_Weapon_GrappleParent!(
    execute
    Object=GetEquippedWeapon!(self:Fort Weapon Reticle Extension Widget Base Object Reference) -> (ReturnValue:Fort Weapon Object Reference)
   ) -> (then, CastFailed, AsB Weapon Grapple Parent:B Weapon Grapple Parent Object Reference) {is=title K2Node_DynamicCast}
   > Bind Event to BP Reticle Update(self:Fort Weapon Ranged Dual Swing Object Reference=@n29, Delegate=@n2) {is=title K2Node_AddDelegate}
   > Bind Event to Reticle Cooldown(self:B Weapon Grapple Parent Object Reference=@n29, Delegate=@n1) {is=title K2Node_AddDelegate}
  ]}
 GetTargetSize {nodes [
   n1 = Inputs!() -> (execute, PhysicsObject:Actor Object Reference)
   > n2 = GetPhysicsObjectSimSize!(execute, PhysicsObject:Actor Object Reference=@n1) -> (then, ReturnValue:EFortPhysicsSimSize Enum=Small)
   n3 = Cast To FortPickup!(Object=@n1) -> (AsFort Pickup:Fort Pickup Object Reference, bSuccess:bool=~) {is=title K2Node_DynamicCast}
   n4 = Cast To FortProjectileBase!(Object=@n1) -> (AsFort Projectile Base:Fort Projectile Base Object Reference, bSuccess:bool=~) {is=title K2Node_DynamicCast}
   > Outputs!(
    then:Exec
    ReturnValue:EFortPhysicsSimSize Enum=Select(Option 0:EFortPhysicsSimSize Enum=@n2, Option 1:EFortPhysicsSimSize Enum=Small, Index=OR Boolean(A=@n3.bSuccess, B=@n4.bSuccess)) -> (ReturnValue:EFortPhysicsSimSize Enum)
   )
  ]}
]
variables [
 AbleToUse {type=boolean}
 Attached {type=boolean}
 Cooldown_GE {type=ActiveGameplayEffectHandle, default {bPassedFiltersAndWasExecuted=false}}
 CooldownAnimPlayed {type=boolean}
 CooldownInterval {type=number}
 CooldownRingDynamicMaterial {type=MaterialInstanceDynamic}
 CooldownStartTime {type=number}
 CooldownTime {type=number}
 CooldownTimerHandle {type=TimerHandle}
 CurrentCooldownDuration {type=number}
 FullCooldown {type=number, default=8}
 FullCooldownTime {type=ScalableFloat, default {registryType=None, value=0, curve {curveTable=null, rowName=None}}}
 HasPlayedAttachedAnim {type=boolean}
 HasPlayedCantUseAnim {type=boolean}
 HasPlayedLockedOnAnim {type=boolean}
 IsCooldownPlaying {type=boolean}
 IsFiringBlocked {type=boolean}
 Locked On {type=boolean}
 OnCooldown {type=boolean}
 Player {type=Actor, default=null}
 TargetPosition {type=Vector}
 TotalCooldownDuration {type=number}
]
widgets {
 info {
  inheritedWidgetCount=1
  parentClass=/Script/FortniteUI.FortWeaponReticleExtensionWidgetBase
  rootWidgetClass=&Overlay
  widgetCount=10
 }
 widgets [
  Reticle {widgetClassPath=&Overlay, widget=Reticle}
  CooldownRing {
   widgetClassPath=&Image
   bIsVariable=true
   parent=Reticle
   widget=CooldownRing
   properties {
    visibility=HitTestInvisible
    brush {
     bIsDynamicallyLoaded=false
     drawAs=Image
     imageType=NoImage
     mirroring=NoMirror
     resourceName=None
     resourceObject=&/M_UI_WestSausage_Reticle_Ring
     tiling=NoTile
     imageSize=<56 56>
     margin=<ltrb 0 0 0 0>
     outlineSettings {
      bUseBrushTransparency=false
      roundingType=HalfHeightRadius
      width=0
      color {colorUseRule=UseColor_Specified, specifiedColor=<rgba 0 0 0 0>}
      cornerRadii=<0 0 0 0>
     }
     tintColor {colorUseRule=UseColor_Specified, specifiedColor=<rgba 1 1 1 1>}
     uVRegion {bIsValid=false, max=<0 0>, min=<0 0>}
    }
   }
   slot {class=OverlaySlot, properties {horizontalAlignment=HAlign_Center, verticalAlignment=VAlign_Center}}
  }
  AnimBoundSizeControl {
   widgetClassPath=&Border
   parent=Reticle
   widget=AnimBoundSizeControl
   properties {
    visibility=SelfHitTestInvisible
    background {
     bIsDynamicallyLoaded=false
     drawAs=NoDrawType
     imageType=NoImage
     mirroring=NoMirror
     resourceName=None
     resourceObject=null
     tiling=NoTile
     imageSize=<32 32>
     margin=<ltrb 0 0 0 0>
     outlineSettings {
      bUseBrushTransparency=false
      roundingType=HalfHeightRadius
      width=0
      color {colorUseRule=UseColor_Specified, specifiedColor=<rgba 0 0 0 0>}
      cornerRadii=<0 0 0 0>
     }
     tintColor {colorUseRule=UseColor_Specified, specifiedColor=<rgba 1 1 1 1>}
     uVRegion {bIsValid=false, max=<0 0>, min=<0 0>}
    }
    padding=<ltrb 0 0 0 0>
    renderTransform {angle=0, scale=<1.1875 1.1875>, shear=<0 0>, translation=<0 0>}
   }
   slot {class=OverlaySlot, properties {horizontalAlignment=HAlign_Center, verticalAlignment=VAlign_Center}}
  }
  Bracket {
   widgetClassPath=&Overlay
   bIsVariable=true
   parent=AnimBoundSizeControl
   widget=Bracket
   slot {class=BorderSlot, properties {padding=<ltrb 0 0 0 0>}}
  }
  Reticle_Surrounding_Shadow {
   widgetClassPath=&Image
   bIsVariable=true
   parent=Bracket
   widget=Reticle_Surrounding_Shadow
   properties {
    visibility=SelfHitTestInvisible
    brush {
     bIsDynamicallyLoaded=false
     drawAs=Image
     imageType=NoImage
     mirroring=NoMirror
     resourceName=None
     resourceObject=&/M_UI_WestSausage_ReticleBracket_Shadow
     tiling=NoTile
     imageSize=<96 96>
     margin=<ltrb 0 0 0 0>
     outlineSettings {
      bUseBrushTransparency=false
      roundingType=HalfHeightRadius
      width=0
      color {colorUseRule=UseColor_Specified, specifiedColor=<rgba 0 0 0 0>}
      cornerRadii=<0 0 0 0>
     }
     tintColor {colorUseRule=UseColor_Specified, specifiedColor=<rgba 0 0 0 1>}
     uVRegion {bIsValid=false, max=<0 0>, min=<0 0>}
    }
   }
   slot {class=OverlaySlot, properties {horizontalAlignment=HAlign_Center, verticalAlignment=VAlign_Center}}
  }
  Reticle_Surrounding_BaseDarken {
   widgetClassPath=&Image
   bIsVariable=true
   parent=Bracket
   widget=Reticle_Surrounding_BaseDarken
   properties {
    visibility=SelfHitTestInvisible
    brush {
     bIsDynamicallyLoaded=false
     drawAs=Image
     imageType=NoImage
     mirroring=NoMirror
     resourceName=None
     resourceObject=&/M_UI_WestSausage_ReticleBracket_BaseDarken
     tiling=NoTile
     imageSize=<96 96>
     margin=<ltrb 0 0 0 0>
     outlineSettings {
      bUseBrushTransparency=false
      roundingType=HalfHeightRadius
      width=0
      color {colorUseRule=UseColor_Specified, specifiedColor=<rgba 0 0 0 0>}
      cornerRadii=<0 0 0 0>
     }
     tintColor {colorUseRule=UseColor_Specified, specifiedColor=<rgba 1 1 1 1>}
     uVRegion {bIsValid=false, max=<0 0>, min=<0 0>}
    }
   }
   slot {class=OverlaySlot}
  }
  Reticle_Surrounding_Disabled {
   widgetClassPath=&Image
   bIsVariable=true
   parent=Bracket
   widget=Reticle_Surrounding_Disabled
   properties {
    visibility=Hidden
    brush {
     bIsDynamicallyLoaded=false
     drawAs=Image
     imageType=NoImage
     mirroring=NoMirror
     resourceName=None
     resourceObject=&/M_UI_WestSausage_ReticleBracket_Disabled
     tiling=NoTile
     imageSize=<96 96>
     margin=<ltrb 0 0 0 0>
     outlineSettings {
      bUseBrushTransparency=false
      roundingType=HalfHeightRadius
      width=0
      color {colorUseRule=UseColor_Specified, specifiedColor=<rgba 0 0 0 0>}
      cornerRadii=<0 0 0 0>
     }
     tintColor {colorUseRule=UseColor_Specified, specifiedColor=<rgba 1 1 1 1>}
     uVRegion {bIsValid=false, max=<0 0>, min=<0 0>}
    }
   }
   slot {class=OverlaySlot, properties {horizontalAlignment=HAlign_Fill, verticalAlignment=VAlign_Fill}}
  }
  Reticle_Surrounding {
   widgetClassPath=&Image
   bIsVariable=true
   parent=Bracket
   widget=Reticle_Surrounding
   properties {
    visibility=SelfHitTestInvisible
    brush {
     bIsDynamicallyLoaded=false
     drawAs=Image
     imageType=NoImage
     mirroring=NoMirror
     resourceName=None
     resourceObject=&/M_UI_WestSausage_ReticleBracket
     tiling=NoTile
     imageSize=<96 96>
     margin=<ltrb 0 0 0 0>
     outlineSettings {
      bUseBrushTransparency=false
      roundingType=HalfHeightRadius
      width=0
      color {colorUseRule=UseColor_Specified, specifiedColor=<rgba 0 0 0 0>}
      cornerRadii=<0 0 0 0>
     }
     tintColor {colorUseRule=UseColor_Specified, specifiedColor=<rgba 1 1 1 1>}
     uVRegion {bIsValid=false, max=<0 0>, min=<0 0>}
    }
   }
   slot {class=OverlaySlot}
  }
  Reticle_Spinners_AttachedFast {
   widgetClassPath=&Image
   bIsVariable=true
   parent=Bracket
   widget=Reticle_Spinners_AttachedFast
   properties {
    renderOpacity=0
    visibility=Hidden
    brush {
     bIsDynamicallyLoaded=false
     drawAs=Image
     imageType=NoImage
     mirroring=NoMirror
     resourceName=None
     resourceObject=&/M_UI_WestSausage_Reticle_BracketSpinningCircles_Fast
     tiling=NoTile
     imageSize=<56 56>
     margin=<ltrb 0 0 0 0>
     outlineSettings {
      bUseBrushTransparency=false
      roundingType=HalfHeightRadius
      width=0
      color {colorUseRule=UseColor_Specified, specifiedColor=<rgba 0 0 0 0>}
      cornerRadii=<0 0 0 0>
     }
     tintColor {colorUseRule=UseColor_Specified, specifiedColor=<rgba 1 1 1 1>}
     uVRegion {bIsValid=false, max=<0 0>, min=<0 0>}
    }
   }
   slot {class=OverlaySlot, properties {horizontalAlignment=HAlign_Center, verticalAlignment=VAlign_Center}}
  }
  TutorialHighlight {widgetClassPath=&TutorialHighlight}
 ]
}
