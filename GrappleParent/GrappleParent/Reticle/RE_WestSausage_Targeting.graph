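# Targeting reticle widget, built on CommonUserWidget, for the WestSausage item in the GrappleParent content.
# Its centered overlay stacks a translucent targeting-circle material image and a nested LockedOnReticleWidget.
# The UpdateChargeCircle event offsets and multiplies an incoming scale value, then applies it uniformly as render scale to the circle's wrap box.
# Targeting state variables such as Locked On, Attached and TargetPosition are declared but unused in this graph.
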WidgetBlueprint /GrappleParent/GrappleParent/Reticle/RE_WestSausage_Targeting
graphs [EventGraph {nodes [
   n1 = UpdateChargeCircle!() -> (OutputDelegate, then, Scale X:double=~) {is=title K2Node_CustomEvent}
   n2 = float * float(A=float + float(A=@n1.Scale X, B=0.2), B=4) {is=title K2Node_PromotableOperator}
   > SetRenderScale!(execute, self:Widget Object Reference=Get WrapBox_1() -> (WrapBox_1:Wrap Box Object Reference), Scale_X:double=@n2, Scale_Y:double=@n2) -> (then)
  ]}]
variables [
 AbleToUse {type=boolean}
 Attached {type=boolean}
 Cached_AboverOverloadThreshold {type=boolean}
 Cast Hit Something {type=boolean}
 Locked On {type=boolean}
 Player {type=Actor, default=null}
 PosInViewport {type=Vector2D, default=<0 0>}
 TargetPosition {type=Vector}
]
widgets {
 info {inheritedWidgetCount=0, parentClass=&CommonUserWidget, rootWidgetClass=&Overlay, widgetCount=7}
 widgets [
  Reticle {widgetClassPath=&Overlay, bIsVariable=true, widget=Reticle}
  LockedOn {
   widgetClassPath=&SizeBox
   parent=Reticle
   widget=LockedOn
   properties {heightOverride=96, widthOverride=96}
   slot {class=OverlaySlot, properties {horizontalAlignment=HAlign_Center, verticalAlignment=VAlign_Center}}
  }
  Overlay_LockedOn {
   widgetClassPath=&Overlay
   bIsVariable=true
   parent=LockedOn
   widget=Overlay_LockedOn
   slot {class=SizeBoxSlot, properties {horizontalAlignment=HAlign_Center, verticalAlignment=VAlign_Center}}
  }
  WrapBox_1 {
   widgetClassPath=&WrapBox
   bIsVariable=true
   parent=Overlay_LockedOn
   widget=WrapBox_1
   slot {class=OverlaySlot, properties {horizontalAlignment=HAlign_Center}}
  }
  Image_134 {
   widgetClassPath=&Image
   bIsVariable=true
   parent=WrapBox_1
   widget=Image_134
   properties {
    brush {
     bIsDynamicallyLoaded=false
     drawAs=Image
     imageType=NoImage
     mirroring=NoMirror
     resourceName=None
     resourceObject=&/M_UI_WestSausage_Reticle_TargetingCircle
     tiling=NoTile
     imageSize=<50 50>
     margin=<ltrb 0 0 0 0>
     outlineSettings {
      bUseBrushTransparency=false
      roundingType=HalfHeightRadius
      width=0
      color {colorUseRule=UseColor_Specified, specifiedColor=<rgba 0 0 0 0>}
      cornerRadii=<0 0 0 0>
     }
     tintColor {colorUseRule=UseColor_Specified, specifiedColor=<rgba 1 1 1 0.521>}
     uVRegion {bIsValid=false, max=<0 0>, min=<0 0>}
    }
    renderTransform {angle=0, scale=<1 1>, shear=<0 0>, translation=<-9.5 -9.5>}
   }
   slot {class=WrapBoxSlot, properties {horizontalAlignment=HAlign_Center, verticalAlignment=VAlign_Center}}
  }
  WrapBox_0 {
   widgetClassPath=&WrapBox
   parent=Overlay_LockedOn
   widget=WrapBox_0
   slot {class=OverlaySlot, properties {horizontalAlignment=HAlign_Center, verticalAlignment=VAlign_Center}}
  }
  Reticle_LockedOn {
   widgetClassPath=&/LockedOnReticleWidget
   bIsVariable=true
   parent=WrapBox_0
   widget=Reticle_LockedOn
   properties {renderTransform {angle=0, scale=<2 2>, shear=<0 0>, translation=<0 0>}}
   slot {class=WrapBoxSlot}
  }
 ]
}
